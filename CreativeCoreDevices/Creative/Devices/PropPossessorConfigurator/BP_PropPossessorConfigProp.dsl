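; Creative device that configures Prop Hunt possession rules for the current player-created minigame.
; On BeginPlay and minigame start, only where cosmetic events cannot run, it registers itself as the minigame's PropHuntSettings and updates the prop possessor item's drop override.
; SetMinigameDropRule adds, changes or removes that map entry from the drop-behavior option.
; Getter functions expose possession, ping and equip options, while animation-override and filtering getters return fixed defaults.
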(event EventBeginPlay
  (Visibility|SetVisibilityComponents (Variables|Default|GetVisibilityRulesComponent) (Utilities|Array|MakeArray (Variables|Debug|GetStaticMeshComponent) (Variables|Default|GetSkeletalMesh)) true)
  (Networking|CanExecuteCosmeticEvents
    (:False
      (bind _asminigame_player_created (Utilities|Casting|CastToMinigame_PlayerCreated (Minigame|GetMinigame self :!class /Script/FortniteGame.MinigameBlueprintLibrary)))
      (Variables|/Game/Athena/Playset/Minigames/Minigame_PlayerCreated.Minigame_PlayerCreated_C|SetPropHuntSettings self _asminigame_player_created)
      (CallFunction|SetMinigameDropRule _asminigame_player_created))))

(event MinigameStartedDelegate(FortMinigameProgress)
  (Networking|CanExecuteCosmeticEvents
    (:False
      (bind _asminigame_player_created (Utilities|Casting|CastToMinigame_PlayerCreated (Minigame|GetMinigame self :!class /Script/FortniteGame.MinigameBlueprintLibrary)))
      (Variables|/Game/Athena/Playset/Minigames/Minigame_PlayerCreated.Minigame_PlayerCreated_C|SetPropHuntSettings self _asminigame_player_created)
      (CallFunction|SetMinigameDropRule _asminigame_player_created))))

(fn EquipPrimaryItemAfterCancelPossession
  (return (!= (Variables|Default|GetEquipPrimaryItemAfterCancelPropPossession) 0)))

(fn GetDisguiseAnimationOverrideDetails
  (return false 0.0))

(fn GetPingDetails
  (return (Variables|Default|GetShowPingTimer) (Math|Conversions|ToFloat(Integer) (Variables|Default|GetPingFrequency))))

(fn GetPropDetails
  (return (!= (Variables|Default|GetShowPropsRemaining) 0) (!= (Variables|Default|GetPropHealthSetting) 0) (!= (Variables|Default|GetAllowChangePossessedProp) 0) (!= (Variables|Default|GetAllowCancelPropPossession) 0)))

(fn GetFilteringDetails
  (return 0))

(fn SetMinigameDropRule
  (params (Minigame :type Minigame%20Player%20Created%20Object%20Reference))
  (bind _proppossessoritemdef (Variables|Default|GetPropPossessorItemDef))
  (bind _minigame (Variables|GetMinigame))
  (bind _itemdropoverride (Variables|/Game/Athena/Playset/Minigames/Minigame_PlayerCreated.Minigame_PlayerCreated_C|GetItemDropOverride _minigame))
  (bind (_value _returnvalue) (Utilities|Map|Find _itemdropoverride _proppossessoritemdef))
  (Utilities|IsValid Minigame
    (:Is%20Valid
      (LocalVariables|SetDropBehavior (Utilities|Select :Index (Variables|Default|GetDropPropGunBehavior) :Option%200 "NewEnumerator0" :Option%201 "NewEnumerator0" :Option%202 "NewEnumerator1" :Option%203 "NewEnumerator2"))
      (LocalVariables|SetOverrideDefault (!= (Variables|Default|GetDropPropGunBehavior) 0))
      (if _returnvalue
        (if (LocalVariables|GetOverrideDefault)
          (Variables|SetBy-RefVar _value (LocalVariables|GetDropBehavior))
          (else
            (Utilities|Map|Remove (Variables|/Game/Athena/Playset/Minigames/Minigame_PlayerCreated.Minigame_PlayerCreated_C|GetItemDropOverride (Variables|GetMinigame)) (Variables|Default|GetPropPossessorItemDef))))
        (else
          (if (LocalVariables|GetOverrideDefault)
            (Utilities|Map|Add (Variables|/Game/Athena/Playset/Minigames/Minigame_PlayerCreated.Minigame_PlayerCreated_C|GetItemDropOverride (Variables|GetMinigame)) (Variables|Default|GetPropPossessorItemDef) (LocalVariables|GetDropBehavior))))))))

(fn ConstructionScript)
